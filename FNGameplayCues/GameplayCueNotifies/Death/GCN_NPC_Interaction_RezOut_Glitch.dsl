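; Gameplay cue notify for an NPC glitch rez-out on death, extending a parent cue's dissolve VFX event.
; After the parent logic runs, it checks the spawned death VFX and rotates it to match the pawn mesh's world rotation.
; It then plays the NPC death sound at the VFX location, with a leftover debug print in between.
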(event EventSpawnDissolveVFX
  (bind _pawn (Variables|DONOTEDIT|GetPawn))
  (|CallParentFunction)
  (bind _spawned_death_vfx (Variables|DONOTEDIT|GetSpawned%20Death%20VFX)
    (:then
      (Transformation|SetWorldRotation _spawned_death_vfx (Transformation|GetWorldRotation (Variables|/Script/Engine.Character|GetMesh _pawn)) false false)
      (Development|PrintString 0 "124151")
      (Audio|PlaySoundatLocation _pawn (Variables|Default|GetSoundonNPCDeath) (Transformation|GetWorldLocation _spawned_death_vfx)))
    (:else)))

(fn ConstructionScript
  (|CallParentFunction))
