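; GCN_RezIn is a gameplay cue notify whose helper functions build and clean up a player's teleport-in effect.
; They spawn an optional teleportation drone with its sound library, Niagara effects fed the pawn's mesh, and a point light skipped on mobile.
; Other helpers set dissolve and glow timeline rates, fade falling and storm audio, and restore materials from the pawn's saved list using a running index across meshes.
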(fn AddSoundLibraryForTeleportation
  (bind _drone (Variables|DONOTEDIT|GetDrone)
    (:then
      ; add specific library for warmup sound
      (SoundLibrary|AddLibrary (WorldSubsystems|GetSoundLibrarySubsystem _drone) _drone "/Game/Sounds/Fort_Gadgets/Teleporter/SL_Teleporter_RezIn.SL_Teleporter_RezIn" :!class /Script/SoundLibrary.SoundLibrarySubsystem)
      ; remove front end library
      (SoundLibrary|RemoveLibrary (WorldSubsystems|GetSoundLibrarySubsystem _drone) _drone "/Game/Sounds/Fort_Gadgets/Teleporter/SL_Teleporter_RezIn_FrontEnd.SL_Teleporter_RezIn_FrontEnd" :!class /Script/SoundLibrary.SoundLibrarySubsystem))
    (:else)))

(fn Clean-UpTeleportationLight
  (bind _teleportation_point_light (Variables|POINTLIGHT|GetTeleportation%20Point%20Light)
    (:then
      (Components|DestroyComponent _teleportation_point_light 0)
      (Variables|POINTLIGHT|SetTeleportation%20Point%20Light 0))
    (:else
      _))
  (return))

(fn RemoveSoundLibraryForTeleportation
  (bind _drone (Variables|DONOTEDIT|GetDrone)
    (:then
      ; remove library after spawn complete
      (SoundLibrary|RemoveLibrary (WorldSubsystems|GetSoundLibrarySubsystem _drone) _drone "/Game/Sounds/Fort_Gadgets/Teleporter/SL_Teleporter_RezIn.SL_Teleporter_RezIn" :!class /Script/SoundLibrary.SoundLibrarySubsystem))
    (:else)))

(fn RestoreCharacterMaterials
  (bind _pawn (Variables|DONOTEDIT|GetPawn))
  (TeleportationFX|CleanupDissolveFX _pawn (Variables|DISSOLVE|GetMat%20Chracter%20Dissolve) (Variables|DONOTEDIT|GetDissolveMIDs))
  (for _array_element (Variables|/Game/Athena/PlayerPawn_Athena_Generic_Parent.PlayerPawn_Athena_Generic_Parent_C|GetSkeletalMeshes _pawn)
    (for (_ _array_index) (Rendering|Material|GetMaterials _array_element)
      (LocalVariables|SetCurrentReplacementIndex (+ _array_index (LocalVariables|GetItr)))
      (Rendering|Material|SetMaterial _array_element _array_index (select (Utilities|Array|IsValidIndex (Variables|/Game/Athena/PlayerPawn_Athena_Generic_Parent.PlayerPawn_Athena_Generic_Parent_C|GetPawnMaterials_ALL (Variables|DONOTEDIT|GetPawn)) (LocalVariables|GetCurrentReplacementIndex)) (Utilities|Array|Get(acopy) (Variables|/Game/Athena/PlayerPawn_Athena_Generic_Parent.PlayerPawn_Athena_Generic_Parent_C|GetPawnMaterials_ALL (Variables|DONOTEDIT|GetPawn)) (LocalVariables|GetCurrentReplacementIndex)) 0)))
    (LocalVariables|SetItr (+ (LocalVariables|GetItr) (Utilities|Array|Length (Rendering|Material|GetMaterials _array_element)))))
  (return true))

(fn SetTimelines'Playrates
  (Components|Timeline|SetPlayRate (Variables|GCN_RezIn|GetTFX_ResOutCharacterMesh) (Variables|DISSOLVE|GetDissolve%20Timeline%20Playrate))
  (Components|Timeline|SetPlayRate (Variables|GCN_RezIn|GetTFX_GlowCharacterMesh) (Variables|DISSOLVE|GetGlow%20Timeline%20Playrate)))

(fn SpawnDroneVFX
  (bind _returnvalue (Niagara|SpawnSystemAttached (Variables|Default|GetDrone%20Visual%20Effect) (Variables|/Game/Blueprints/Adventures/SpawnTeleporter/BP_TeleportationDrone.BP_TeleportationDrone_C|GetSkeletalMeshComponentBudgeted (Variables|DONOTEDIT|GetDrone)) (Variables|Default|GetDrone%20VFX%20Attach%20Point) "0, 0, 0" "0, 0, 0" "SnapToTarget" false false))
  (Niagara|SetNiagaraSkeletalMeshComponent _returnvalue "SkeletalMesh" (Variables|/Script/Engine.Character|GetMesh (Variables|DONOTEDIT|GetPawn)))
  (CallFunction|SetAdditionalNiagaraParameterOnDroneFX _returnvalue)
  (Components|Activation|Activate _returnvalue false))

(fn SpawnTeleportInVFX
  (if (Variables|DISSOLVE|GetSpawn%20VFX%20Attached)
    (bind _returnvalue (Niagara|SpawnSystemAttached (Variables|DISSOLVE|GetTeleport%20In%20Visual%20Effect) (Variables|/Script/Engine.Character|GetMesh (Variables|DONOTEDIT|GetPawn)) (Variables|DISSOLVE|GetTeleport%20In%20VFX%20Attach%20Point%20Name) "0, 0, 0" "0, 0, 0" "SnapToTarget" true))
    (Variables|DONOTEDIT|SetSpawned%20Teleport%20VFX _returnvalue)
    (else
      (bind _returnvalue_2 (Niagara|SpawnSystematLocation 0 (Variables|DISSOLVE|GetTeleport%20In%20Visual%20Effect) (Transformation|GetSocketLocation (Variables|/Script/Engine.Character|GetMesh (Variables|DONOTEDIT|GetPawn)) (Variables|DISSOLVE|GetTeleport%20In%20VFX%20Attach%20Point%20Name))))
      (bind _output_get (Variables|DONOTEDIT|SetSpawned%20Teleport%20VFX _returnvalue_2))
      (Niagara|SetNiagaraSkeletalMeshComponent _output_get "SkeletalMesh" (Variables|/Script/Engine.Character|GetMesh (Variables|DONOTEDIT|GetPawn)))))
  (CallFunction|SetAdditionalNiagaraParametersOnSpawnedFX (Variables|DONOTEDIT|GetSpawned%20Teleport%20VFX))
  (return))

(fn SpawnTeleportationDrone
  (Variables|DONOTEDIT|SetDrone 0)
  (if (Variables|DRONE|GetSpawnDrone)
    (bind _returnvalue (Game|SpawnActorfromClass self (Variables|DRONE|GetTeleportation%20Drone) (Transformation|GetWorldTransform (Variables|/Script/Engine.Character|GetMesh (Variables|DONOTEDIT|GetPawn))) "AlwaysSpawn" "SelectDefaultAtRuntime" self true (not (GlobalUIContext|IsinZone (BlueprintContext|GetFortGlobalUIContext))) (Variables|DONOTEDIT|GetPawn)))
    (bind _output_get (Variables|DONOTEDIT|SetDrone _returnvalue))
    (Actor|SetLifeSpan _output_get (Variables|DRONE|GetTeleport%20Bot%20Lifespan))
    (Variables|/Game/Blueprints/Adventures/SpawnTeleporter/BP_TeleportationDrone.BP_TeleportationDrone_C|SetAnimPlayRate (Variables|DRONE|GetTeleport_Bot_AnimPlayRate) _output_get))
  (return))

(fn SpawnTeleportationLight
  (if (Fortnite|IsMobilePlatform)
    (Development|PrintString (Variables|DONOTEDIT|GetPawn) "Rez Out - MOBILE PLATFORM - No Point LIght for you.")
    (else
      (Variables|POINTLIGHT|GetTeleportation%20Point%20Light
        (:then)
        (:else
          (bind _returnvalue (AddComponent|NODE_AddPointLightComponent-0 :RelativeTransform (Math|Conversions|ToTransform(Vector) (+ (Transformation|GetSocketLocation (Variables|/Script/Engine.Character|GetMesh (Variables|DONOTEDIT|GetPawn)) "root") (Variables|POINTLIGHT|GetTeleportation%20Light%20Offset)))))
          (bind _output_get (Variables|POINTLIGHT|SetTeleportation%20Point%20Light _returnvalue))
          (Rendering|Components|Light|SetIntensity _output_get 0.0)
          (Rendering|Components|Light|SetLightColor _output_get (Variables|POINTLIGHT|GetTeleportation%20Light%20Color) true))))))

(fn StopLoopingAudio
  (bind _asbpi_player_pawn_athena (Utilities|Casting|CastToBPI_PlayerPawnAthena (Variables|DONOTEDIT|GetPawn)))
  (Utilities|FlowControl|Sequence
    (:then_0
      (bind _fallingaudioloop (BPI_PlayerPawnAthena|GetFallingAudioLoopComponent _asbpi_player_pawn_athena :!class /Game/Athena/BPI_PlayerPawnAthena.BPI_PlayerPawnAthena_C))
      (Utilities|IsValid _fallingaudioloop
        (:Is%20Valid
          (Audio|Components|Audio|FadeOut _fallingaudioloop 1.0 0.0 "Linear" :!class /Script/Engine.AudioComponent))
        (:Is%20Not%20Valid
          _)))
    (:then_1
      (bind _stormaudioloop (BPI_PlayerPawnAthena|GetStormAudioLoopComponent _asbpi_player_pawn_athena :!class /Game/Athena/BPI_PlayerPawnAthena.BPI_PlayerPawnAthena_C))
      (Utilities|IsValid _stormaudioloop
        (:Is%20Valid
          (Audio|Components|Audio|FadeOut _stormaudioloop 1.0 0.0 "Linear" :!class /Script/Engine.AudioComponent))
        (:Is%20Not%20Valid
          _))))
  (return))

(fn ConstructionScript)
