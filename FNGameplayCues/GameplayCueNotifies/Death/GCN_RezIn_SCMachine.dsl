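; Gameplay cue notify for the rez-in effect of an SCMachine respawn, with a visual variant for enhanced resurrection.
; On burst it caches whether the aggregated source tags contain the EnhancedResurrection tag, then calls the parent, which it extends.
; When the parent's hook for extra Niagara parameters on the spawned teleport-in FX fires, it sets that system's Variant parameter if the respawn is enhanced.
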(event EventOnBurst(EitherVFX-preferthisversion)
  (params (MyTarget :type Actor%20Object%20Reference) (Parameters :type Gameplay%20Cue%20Parameters%20Structure%20%28by%20ref%29) (ParticleComponents :type Array%20of%20FXSystem%20Component%20Object%20References) (AudioComponents :type Array%20of%20Audio%20Component%20Object%20References) (BurstCameraShakeInstance :type Camera%20Shake%20Base%20Object%20Reference) (BurstDecalInstance :type Decal%20Actor%20Object%20Reference))
  (bind (_normalizedmagnitude _rawmagnitude _effectcontext _matchedtagname _originaltag _aggregatedsourcetags) (Ability|GameplayCue|BreakGameplayCueParameters Parameters))
  ;[0.03157 0.15 0 0.5] OnBurst: Cache if the respawn is enhanced.
  (Utilities|FlowControl|Sequence
    (:then_0
      (Variables|Transient|SetbEnhancedResurrection (GameplayTags|HasTag _aggregatedsourcetags "(TagName=\"SCMachine.Respawn.EnhancedResurrection\")" true)))
    (:then_1
      (|CallParentFunction MyTarget Parameters ParticleComponents AudioComponents BurstCameraShakeInstance BurstDecalInstance))))

(event EventSetAdditionalNiagaraParametersonSpawnedFX
  (params (Spawned%20Teleport%20In%20FX :type Niagara%20Particle%20System%20Component%20Object%20Reference))
  ;[0 0.1226 0.15 0.5] SetAdditionalNiagaraParametersOnSpawnedFX: If enhanced, update the variant value of the respawn FX.
  (|CallParentFunction Spawned%20Teleport%20In%20FX)
  (if (Variables|Transient|GetbEnhancedResurrection)
    (Utilities|IsValid Spawned%20Teleport%20In%20FX
      (:Is%20Valid
        (Niagara|SetNiagaraVariable(Int32) Spawned%20Teleport%20In%20FX "Variant" 1)))))

(fn ConstructionScript
  (|CallParentFunction))
