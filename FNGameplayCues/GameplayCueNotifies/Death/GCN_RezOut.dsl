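; Burst gameplay cue that plays the rez-out death effect on Athena player pawns.
; It caches hit data, plays a state-based elimination montage, fades looping audio, and spawns an optional drone and non-mobile point light.
; It gathers character and weapon meshes, starts a glow timeline and plays a local-player sound sweetener.
; Helpers handle dissolve materials and Niagara effects, with empty hook events letting variants alter the sequence and effect parameters.
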(event EventOnBurst(EitherVFX-preferthisversion)
  (params (MyTarget :type Actor%20Object%20Reference) (Parameters :type Gameplay%20Cue%20Parameters%20Structure%20%28by%20ref%29) (ParticleComponents :type Array%20of%20FXSystem%20Component%20Object%20References) (AudioComponents :type Array%20of%20Audio%20Component%20Object%20References) (BurstCameraShakeInstance :type Camera%20Shake%20Base%20Object%20Reference) (BurstDecalInstance :type Decal%20Actor%20Object%20Reference))
  (bind _dissolve (Variables|DONOTEDIT|GetDissolve))
  (CallFunction|SetonDeathHitParameters Parameters)
  (Variables|DONOTEDIT|SetParticle%20Components ParticleComponents)
  (Variables|DONOTEDIT|SetPawn 0)
  (Utilities|Set|Clear (Variables|DONOTEDIT|GetDissolve))
  (Utilities|IsA(soft) MyTarget "/Game/Athena/PlayerPawn_Athena.PlayerPawn_Athena_C"
    (:True
      (bind _asplayer_pawn_athena_generic_parent (Utilities|Casting|CastToPlayerPawn_Athena_Generic_Parent MyTarget)
        (:then
          (Variables|DONOTEDIT|SetPawn _asplayer_pawn_athena_generic_parent)
          (CallFunction|CheckImpactLocationForErrors)
          (CallFunction|PlayElminationAnimMontage)
          (CallFunction|StopLoopingAudio)
          (if (Variables|DISSOLVE|GetShouldDelayBetweenAnimAndVFX)
            (Utilities|FlowControl|Delay (Variables|DONOTEDIT|GetPawn) (Variables|DISSOLVE|GetDelayBetweenAnimAndVFX)) ; Add a Potential Delay for IPs (Such as Realm/Vader) where we want to see him animate more when he dies.
          )
          (CallFunction|SpawnTeleportationDrone)
          (CallFunction|SpawnTeleportationLight)
          (CallFunction|Pre-SequenceChange)
          ;[1 1 1] Get all Meshes into one Set
          (Utilities|Set|AddItems _dissolve (Variables|/Game/Athena/PlayerPawn_Athena_Generic_Parent.PlayerPawn_Athena_Generic_Parent_C|GetSkeletalMeshes (Variables|DONOTEDIT|GetPawn)))
          (bind _asbpi_player_pawn_athena (Utilities|Casting|CastToBPI_PlayerPawnAthena (Variables|DONOTEDIT|GetPawn)))
          (bind _currentweaponmeshes (BPI_PlayerPawnAthena|GetCurrentWeaponMeshComponents _asbpi_player_pawn_athena :!class /Game/Athena/BPI_PlayerPawnAthena.BPI_PlayerPawnAthena_C))
          (Utilities|Set|AddItems _dissolve _currentweaponmeshes)
          (bind _result (Utilities|Set|ToArray _dissolve))
          (Variables|DONOTEDIT|SetMeshes%20to%20Dissolve _result)
          (CallFunction|SetTimelinePlayrates)
          (Components|Timeline|PlayfromStart (Variables|GCN_RezOut|GetTFX_GlowCharacterMesh))
          (CallFunction|SpawnSweetenerifLocal))
        (:CastFailed
          _)))
    (:False
      _))
  (Development|PrintString MyTarget "CAST FAILED - to PlayerPawn_Athena"))

(event Custom|Pre-Sequence%20Change)

(event Custom|Set%20Additional%20Niagara%20Parameters%20on%20Spawned%20FX
  (params (Spawned%20Death%20FX :type Niagara%20Particle%20System%20Component%20Object%20Reference)))

(event Custom|Set%20Additional%20Niagara%20Parameters%20on%20Drone%20FX
  (params (Drone%20FX :type Niagara%20Particle%20System%20Component%20Object%20Reference)))

(event EventTriggerVFX
  ;[1 1 1] Alteration Events to alter Sequence without needing to re-write large sections
  (CallFunction|SpawnDroneVFX))

(fn CheckImpactLocationforErrors
  (if (or (<= (Math|Vector|VectorLength (Variables|OnDeathParameters|GetOn%20Death%20Hit%20Location)) 0) (> (Math|Vector|Distance(Vector) (Collision|GetClosestPointonCollision (Variables|/Script/Engine.Character|GetCapsuleComponent (Variables|DONOTEDIT|GetPawn)) (Variables|OnDeathParameters|GetOn%20Death%20Hit%20Location) "None") (Variables|OnDeathParameters|GetOn%20Death%20Hit%20Location)) 2.0))
    (bind _output_get (Variables|OnDeathParameters|SetOn%20Death%20Hit%20Location (+ (Transformation|GetActorLocation (Variables|DONOTEDIT|GetPawn)) "0.000000,0.000000,75.000000")))
    (Development|PrintString 0 (Utilities|String|Append "Updated Hit Location - " (Utilities|String|ToString(Vector) _output_get)) (Variables|DEBUG|GetDEBUG_IMACTDATA) (Variables|DEBUG|GetDEBUG_IMACTDATA) "(R=0.000000,G=0.660000,B=1.000000,A=1.000000)" 3e+01)))

(fn Clean-UpTeleportationLight
  (bind _teleportation_point_light (Variables|POINTLIGHT|GetTeleportation%20Point%20Light)
    (:then
      (Components|DestroyComponent _teleportation_point_light 0)
      (Variables|POINTLIGHT|SetTeleportation%20Point%20Light 0))
    (:else
      _))
  (return))

(fn DissolveMaterialSetup
  (for _array_element (Variables|DONOTEDIT|GetMeshes%20to%20Dissolve)
    (for _index (range (Rendering|Material|GetNumMaterials _array_element))
      (bind _returnvalue (Rendering|Material|CreateDynamicMaterialInstance _array_element (Variables|DISSOLVE|GetMat%20Chracter%20Dissolve) "None" "None" :!class /Script/Engine.KismetMaterialLibrary))
      (Rendering|Material|SetMaterial _array_element _index _returnvalue)))
  (return))

(fn PlayElminationAnimMontage
  (bind _pawn (Variables|DONOTEDIT|GetPawn))
  (bind _returnvalue (Health|WasDBNOOnDeath _pawn))
  (bind _returnvalue_2 (Parachute|IsSkydiving _pawn))
  (bind _returnvalue_3 (WaterBody|IsSurfaceSwimming _pawn))
  (bind _returnvalue_4 (and _returnvalue _returnvalue_3))
  (bind _dbno_elimination_montage (Variables|ELIMINATIONANIMS|GetDBNO%20Elimination%20Montage))
  (bind _base_elimination_montage (Variables|ELIMINATIONANIMS|GetBase%20Elimination%20Montage))
  (bind _returnvalue_5 (select _returnvalue _dbno_elimination_montage _base_elimination_montage))
  (bind _dbno_swimming_elimination_montage (Variables|ELIMINATIONANIMS|GetDBNO%20Swimming%20Elimination%20Montage))
  (bind _swimming_elimination_montage (Variables|ELIMINATIONANIMS|GetSwimming%20Elimination%20Montage))
  (bind _skydiving_elimination_montage (Variables|ELIMINATIONANIMS|GetSkydiving%20Elimination%20Montage))
  (bind _returnvalue_6 (select _returnvalue_2 _skydiving_elimination_montage _returnvalue_5))
  (bind _returnvalue_7 (select _returnvalue_3 _swimming_elimination_montage _returnvalue_6))
  (bind _returnvalue_8 (select _returnvalue_4 _dbno_swimming_elimination_montage _returnvalue_7))
  (Animation|PlayLocalAnimMontage (Variables|DONOTEDIT|GetPawn) _returnvalue_8 1.0 "None" false)
  (if (Variables|DEBUG|GetDEBUG_ANIMATION)
    (Development|PrintText (Variables|DONOTEDIT|GetPawn) (Utilities|Text|FormatText "NSLOCTEXT(\"[74DE5189FB6055A36DA6B27FC79703F2]\", \"B97BD1C14828FB844F9EC8916C11BED5\", \"PlayLocalAnimMontage IsDBNO{is} WasDBNO{was} Swimming{swimming} Montage{montage}\")" (Utilities|Text|ToText(Boolean) (Variables|/Script/FortniteGame.FortPawn|GetbIsDBNO (Variables|DONOTEDIT|GetPawn))) (Utilities|Text|ToText(Boolean) (Health|WasDBNOOnDeath (Variables|DONOTEDIT|GetPawn))) (Utilities|Text|ToText(Boolean) (WaterBody|IsSurfaceSwimming (Variables|DONOTEDIT|GetPawn))) (Utilities|Text|ToText(Object) _returnvalue_8))))
  (return))

(fn SetOnDeathHitParameters
  (params (Parameters :type Gameplay%20Cue%20Parameters%20Structure%20%28by%20ref%29))
  (bind (_normalizedmagnitude _rawmagnitude _effectcontext _matchedtagname _originaltag _aggregatedsourcetags _aggregatedtargettags _location _normal) (Ability|GameplayCue|BreakGameplayCueParameters Parameters))
  (bind (_returnvalue_bblockinghit _returnvalue_binitialoverlap _returnvalue_time _returnvalue_distance _returnvalue_location _returnvalue_impactpoint _returnvalue_normal _returnvalue_impactnormal _returnvalue_physmat _returnvalue_hitactor) (Collision|BreakHitResult (Ability|EffectContext|GetHitResult _effectcontext)))
  (bind _returnvalue (GameplayTags|HasTag _aggregatedsourcetags "(TagName=\"Gameplay.DamageDone.Headshot\")" true))
  (Variables|OnDeathParameters|SetOn%20Death%20Hit%20Location _returnvalue_location)
  (bind _output_get (Variables|OnDeathParameters|SetOn%20Death%20Hit%20Momentum _normal))
  (Variables|OnDeathParameters|SetbisHeadshot _returnvalue)
  (bind _output_get_2 (Variables|OnDeathParameters|SetOn%20Death%20Hit%20Normal _returnvalue_normal))
  (Development|PrintString 0 (Utilities|String|Append "Hit Location - " (Utilities|String|ToString(Vector) _returnvalue_location) "
" "Hit Normal - " (Utilities|String|ToString(Vector) _output_get_2) "
" "Hit Momentum - " (Utilities|String|ToString(Vector) _output_get) "
" "Is Headshot - " (Utilities|String|ToString(Boolean) _returnvalue) "
" "Hit Actor - " (Utilities|GetDisplayName _returnvalue_hitactor)) (Variables|DEBUG|GetDEBUG_IMACTDATA) (Variables|DEBUG|GetDEBUG_IMACTDATA) "(R=0.000000,G=0.660000,B=1.000000,A=1.000000)" 3e+01))

(fn SetTimelinePlayrates
  (Components|Timeline|SetPlayRate (Variables|GCN_RezOut|GetTFX_ResOutCharacterMesh) (Variables|DISSOLVE|GetDissolve%20Timeline%20Playrate))
  (Components|Timeline|SetPlayRate (Variables|GCN_RezOut|GetTFX_GlowCharacterMesh) (Variables|DISSOLVE|GetGlow%20Timeline%20Playrate)))

(fn SpawnDissolveVFX
  (bind _drone (Variables|DONOTEDIT|GetDrone))
  (if (Variables|DEBUG|GetDEBUG_REMOVESPAWNVFX)
    (Development|PrintString 0 "DEBUG - Removed Spawned VFX")
    (return)
    (else
      (bind _returnvalue (Niagara|SpawnSystematLocation 0 (Variables|DISSOLVE|GetDissolve%20Visual%20Effect) (Transformation|GetSocketLocation (Variables|/Script/Engine.Character|GetMesh (Variables|DONOTEDIT|GetPawn)) (Variables|DISSOLVE|GetDissolve%20VFX%20Spawn%20Point%20Name))))
      (Effects|Components|ParticleSystem|SetIntParameter _returnvalue "Color Variant" (Variables|Default|GetFXColorVariant))
      (bind _output_get (Variables|DONOTEDIT|SetSpawned%20Death%20VFX _returnvalue))
      (Niagara|SetNiagaraSkeletalMeshComponent _output_get "SkeletalMesh" (Variables|/Script/Engine.Character|GetMesh (Variables|DONOTEDIT|GetPawn)))
      (Utilities|IsValid _drone
        (:Is%20Valid
          (Niagara|SetNiagaraVariable(Vector3) _output_get "PointPosition" (select (Variables|DRONE|GetSpawnDrone) (Transformation|GetSocketLocation (Variables|/Game/Blueprints/Adventures/SpawnTeleporter/BP_TeleportationDrone.BP_TeleportationDrone_C|GetSkeletalMeshComponentBudgeted _drone) "BeamOrigin") "0, 0,250.000000")))
        (:Is%20Not%20Valid
          _))
      (CallFunction|SetAdditionalNiagaraParametersOnSpawnedFX (Variables|DONOTEDIT|GetSpawned%20Death%20VFX))
      (return))))

(fn SpawnDroneVFX
  (bind _returnvalue (Niagara|SpawnSystemAttached (Variables|DRONE|GetDrone%20Visual%20Effect) (Variables|/Game/Blueprints/Adventures/SpawnTeleporter/BP_TeleportationDrone.BP_TeleportationDrone_C|GetSkeletalMeshComponentBudgeted (Variables|DONOTEDIT|GetDrone)) (Variables|DRONE|GetDrone%20VFX%20Attach%20Point) "0, 0, 0" "0, 0, 0" "KeepRelativeOffset" true false))
  (Niagara|SetNiagaraSkeletalMeshComponent _returnvalue "SkeletalMesh" (Variables|/Script/Engine.Character|GetMesh (Variables|DONOTEDIT|GetPawn)))
  (CallFunction|SetAdditionalNiagaraParametersOnDroneFX _returnvalue)
  (Components|Activation|Activate _returnvalue false))

(fn SpawnTeleportationDrone
  (Variables|DONOTEDIT|SetDrone 0)
  (if (Variables|DRONE|GetSpawnDrone)
    (bind _returnvalue (Game|SpawnActorfromClass self (Variables|DRONE|GetTeleportation%20Drone) (Transformation|GetWorldTransform (Variables|/Script/Engine.Character|GetMesh (Variables|DONOTEDIT|GetPawn))) "AlwaysSpawn" "SelectDefaultAtRuntime" self false (not (GlobalUIContext|IsinZone (BlueprintContext|GetFortGlobalUIContext))) (Variables|DONOTEDIT|GetPawn)))
    (bind _output_get (Variables|DONOTEDIT|SetDrone _returnvalue))
    (Actor|SetLifeSpan _output_get (Variables|DRONE|GetTeleport%20Bot%20Lifespan))
    (Variables|/Game/Blueprints/Adventures/SpawnTeleporter/BP_TeleportationDrone.BP_TeleportationDrone_C|SetAnimPlayRate (Variables|DRONE|GetTeleport_Bot_AnimPlayRate) _output_get)
    (else
      _))
  (return))

(fn SpawnTeleportationLight
  (if (Variables|DEBUG|GetDEBUG_REMOVESPAWNEDPOINTLIGHT)
    _
    (elif (Fortnite|IsMobilePlatform)
      (Development|PrintString (Variables|DONOTEDIT|GetPawn) "Rez Out - MOBILE PLATFORM - No Point LIght for you.")
      (else
        (Variables|POINTLIGHT|GetTeleportation%20Point%20Light
          (:then)
          (:else
            (bind _returnvalue (AddComponent|NODE_AddPointLightComponent-0 :RelativeTransform (Math|Conversions|ToTransform(Vector) (+ (Transformation|GetSocketLocation (Variables|/Script/Engine.Character|GetMesh (Variables|DONOTEDIT|GetPawn)) "root") (Variables|POINTLIGHT|GetTeleportation%20Light%20Offset)))))
            (bind _output_get (Variables|POINTLIGHT|SetTeleportation%20Point%20Light _returnvalue))
            (Rendering|Components|Light|SetIntensity _output_get 0.0)
            (Rendering|Lighting|SetInverseExposureBlend _output_get 1.0)
            (Rendering|Components|Light|SetLightColor _output_get (Variables|POINTLIGHT|GetTeleportation%20Light%20Color) true)))))))

(fn SpawnSweetenerIfLocal
  (bind _pawn (Variables|DONOTEDIT|GetPawn)
    (:then
      (bind _asbpi_player_pawn_athena (Utilities|Casting|CastToBPI_PlayerPawnAthena _pawn))
      (bind _ondeathcontroller (BPI_PlayerPawnAthena|GetonDeathController _asbpi_player_pawn_athena :!class /Game/Athena/BPI_PlayerPawnAthena.BPI_PlayerPawnAthena_C))
      (Utilities|IsValid _ondeathcontroller
        (:Is%20Valid
          (if (and (Pawn|IsLocalController _ondeathcontroller) (GlobalUIContext|IsinZone (BlueprintContext|GetFortGlobalUIContext)))
            (Audio|SpawnSoundatLocation (Variables|DONOTEDIT|GetPawn) (Variables|OnDeathParameters|GetRezOutSweetener) (Variables|OnDeathParameters|GetOn%20Death%20Hit%20Location))))))
    (:else)))

(fn StopLoopingAudio
  (bind _asbpi_player_pawn_athena (Utilities|Casting|CastToBPI_PlayerPawnAthena (Variables|DONOTEDIT|GetPawn)))
  (Utilities|FlowControl|Sequence
    (:then_0
      (bind _fallingaudioloop (BPI_PlayerPawnAthena|GetFallingAudioLoopComponent _asbpi_player_pawn_athena :!class /Game/Athena/BPI_PlayerPawnAthena.BPI_PlayerPawnAthena_C))
      (Utilities|IsValid _fallingaudioloop
        (:Is%20Valid
          (Audio|Components|Audio|FadeOut _fallingaudioloop 1.0 0.0 "Linear" :!class /Script/Engine.AudioComponent))
        (:Is%20Not%20Valid
          _))
      (return))
    (:then_1
      (bind _stormaudioloop (BPI_PlayerPawnAthena|GetStormAudioLoopComponent _asbpi_player_pawn_athena :!class /Game/Athena/BPI_PlayerPawnAthena.BPI_PlayerPawnAthena_C))
      (Utilities|IsValid _stormaudioloop
        (:Is%20Valid
          (Audio|Components|Audio|FadeOut _stormaudioloop 1.0 0.0 "Linear" :!class /Script/Engine.AudioComponent))
        (:Is%20Not%20Valid
          _)))))

(fn ConstructionScript)
