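; GCN_RezOut_BellyBlast is a gameplay cue notify in the Death folder, apparently for a BellyBlast variant of the rez-out effect.
; Its pre-sequence-change event and construction script only call their parent functions.
; The file adds no logic of its own, so any behavior comes from the parent class.
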(event EventPre-SequenceChange
  (|CallParentFunction))

(fn ConstructionScript
  (|CallParentFunction))
