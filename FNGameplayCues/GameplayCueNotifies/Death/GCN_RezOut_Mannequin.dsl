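; Gameplay cue burst that plays a teleport-style "rez out" death on a FortPlayerMannequin.
; On burst it optionally plays an elimination animation and spawns a teleportation drone and non-mobile point light, then gathers part meshes and their dynamic materials and starts a glow timeline.
; Helpers swap in dissolve materials and spawn Niagara effects bound to the mannequin mesh and drone.
; Empty custom events provide hooks for pre-sequence changes and extra Niagara parameters.
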(event EventOnBurst(EitherVFX-preferthisversion)
  (params (MyTarget :type Actor%20Object%20Reference) (Parameters :type Gameplay%20Cue%20Parameters%20Structure%20%28by%20ref%29) (ParticleComponents :type Array%20of%20FXSystem%20Component%20Object%20References) (AudioComponents :type Array%20of%20Audio%20Component%20Object%20References) (BurstCameraShakeInstance :type Camera%20Shake%20Base%20Object%20Reference) (BurstDecalInstance :type Decal%20Actor%20Object%20Reference))
  (bind _dissolve (Variables|DONOTEDIT|GetDissolve))
  (Variables|DONOTEDIT|SetParticle%20Components ParticleComponents)
  (Variables|DONOTEDIT|SetMannequin 0)
  (Utilities|Set|Clear (Variables|DONOTEDIT|GetDissolve))
  (bind _asfort_player_mannequin (Utilities|Casting|CastToFortPlayerMannequin MyTarget)
    (:then
      (Variables|DONOTEDIT|SetMannequin _asfort_player_mannequin)
      (CallFunction|PlayElminationAnimMontage)
      (CallFunction|SpawnTeleportationDrone)
      (CallFunction|SpawnTeleportationLight)
      (CallFunction|Pre-SequenceChange)
      ;[1 1 1] Get all Meshes into one Set, possible missing weapons if used on mannequin
      (bind _skeletalmeshcomponents (Customization|GetAllPartSkeletalMeshComponents (Variables|DONOTEDIT|GetMannequin)))
      (Utilities|Set|AddItems _dissolve _skeletalmeshcomponents)
      (bind _result (Utilities|Set|ToArray _dissolve))
      (bind _output_get (Variables|DONOTEDIT|SetMeshes%20to%20Dissolve _result))
      (for _array_element _output_get
        (for (_array_element_2 _array_index) (Rendering|Material|GetMaterials _array_element)
          (bind _asmaterial_instance_dynamic (Utilities|Casting|CastToMaterialInstanceDynamic _array_element_2)
            (:then
              (Utilities|Array|Add (Variables|DONOTEDIT|GetDynamicMannequinMIDs) _asmaterial_instance_dynamic))
            (:CastFailed
              (Development|PrintString (Variables|DONOTEDIT|GetMannequin) (Utilities|String|Append "Material is not Dynamic - " (Utilities|String|ToString(Integer) _array_index)))))))
      (CallFunction|SetTimelinePlayrates)
      (Components|Timeline|PlayfromStart (Variables|GCN_RezOut_Mannequin|GetTFX_GlowCharacterMesh)))
    (:CastFailed
      (Development|PrintString 0 "CAST FAILED - to FortPlayerMannequin"))))

(event Custom|Pre-Sequence%20Change)

(event Custom|Set%20Additional%20Niagara%20Parameters%20on%20Spawned%20FX
  (params (Spawned%20Death%20FX :type Niagara%20Particle%20System%20Component%20Object%20Reference)))

(event Custom|Set%20Additional%20Niagara%20Parameters%20on%20Drone%20FX
  (params (Drone%20FX :type Niagara%20Particle%20System%20Component%20Object%20Reference)))

(fn Clean-UpTeleportationLight
  (bind _teleportation_point_light (Variables|POINTLIGHT|GetTeleportation%20Point%20Light)
    (:then
      (Components|DestroyComponent _teleportation_point_light 0)
      (Variables|POINTLIGHT|SetTeleportation%20Point%20Light 0))
    (:else
      _))
  (return))

(fn DissolveMaterialSetup
  (for _array_element (Variables|DONOTEDIT|GetMeshes%20to%20Dissolve)
    (for _index (range (Rendering|Material|GetNumMaterials _array_element))
      (bind _returnvalue (Rendering|Material|CreateDynamicMaterialInstance _array_element (Variables|DISSOLVE|GetMat%20Chracter%20Dissolve) "None" "None" :!class /Script/Engine.KismetMaterialLibrary))
      (Rendering|Material|SetMaterial _array_element _index _returnvalue)))
  (return))

(fn PlayElminationAnimMontage
  (if (Variables|ELIMINATIONANIMS|GetUse%20Elimination%20Montage)
    (Components|Animation|PlayAnimation (Variables|/Script/Engine.SkeletalMeshActor|GetSkeletalMeshComponent (Variables|DONOTEDIT|GetMannequin)) (Variables|ELIMINATIONANIMS|GetBase%20Elimination%20Montage) false)
    (else
      _))
  (if (Variables|DEBUG|GetDEBUG_ANIMATION)
    (Development|PrintText (Variables|DONOTEDIT|GetMannequin) (Utilities|Text|FormatText "NSLOCTEXT(\"[70C03BD97C6EA5E7CEDF0AE105AA289A]\", \"F30AD25B4F6EE465E913F7A94908B14D\", \"Play Animation on Mannequin Skeletal Mesh Component Montage{montage} is set to play {BOOL}\")" (Utilities|Text|ToText(Object) (Variables|ELIMINATIONANIMS|GetBase%20Elimination%20Montage)) (Utilities|Text|ToText(Boolean) (Variables|ELIMINATIONANIMS|GetUse%20Elimination%20Montage)))))
  (return))

(fn SetTimelinePlayrates
  (Components|Timeline|SetPlayRate (Variables|GCN_RezOut_Mannequin|GetTFX_ResOutCharacterMesh) (Variables|DISSOLVE|GetDissolve%20Timeline%20Playrate))
  (Components|Timeline|SetPlayRate (Variables|GCN_RezOut_Mannequin|GetTFX_GlowCharacterMesh) (Variables|DISSOLVE|GetGlow%20Timeline%20Playrate)))

(fn SpawnDissolveVFX
  (bind _drone (Variables|DONOTEDIT|GetDrone))
  (if (Variables|DEBUG|GetDEBUG_REMOVESPAWNVFX)
    (Development|PrintString 0 "DEBUG - Removed Spawned VFX")
    (return)
    (else
      (bind _returnvalue (Niagara|SpawnSystematLocation 0 (Variables|DISSOLVE|GetDissolve%20Visual%20Effect) (Transformation|GetSocketLocation (Variables|/Script/Engine.SkeletalMeshActor|GetSkeletalMeshComponent (Variables|DONOTEDIT|GetMannequin)) (Variables|DISSOLVE|GetDissolve%20VFX%20Spawn%20Point%20Name))))
      (bind _output_get (Variables|DONOTEDIT|SetSpawned%20Death%20VFX _returnvalue))
      (Niagara|SetNiagaraSkeletalMeshComponent _output_get "SkeletalMesh" (Variables|/Script/Engine.SkeletalMeshActor|GetSkeletalMeshComponent (Variables|DONOTEDIT|GetMannequin)))
      (Utilities|IsValid _drone
        (:Is%20Valid
          (Niagara|SetNiagaraVariable(Vector3) _output_get "PointPosition" (select (Variables|DRONE|GetSpawnDrone) (Transformation|GetSocketLocation (Variables|/Game/Blueprints/Adventures/SpawnTeleporter/BP_TeleportationDrone.BP_TeleportationDrone_C|GetSkeletalMeshComponentBudgeted _drone) "BeamOrigin") "0, 0,250.000000")))
        (:Is%20Not%20Valid
          _))
      (CallFunction|SetAdditionalNiagaraParametersOnSpawnedFX (Variables|DONOTEDIT|GetSpawned%20Death%20VFX))
      (return))))

(fn SpawnDroneVFX
  (bind _returnvalue (Niagara|SpawnSystemAttached (Variables|DRONE|GetDrone%20Visual%20Effect) (Variables|/Game/Blueprints/Adventures/SpawnTeleporter/BP_TeleportationDrone.BP_TeleportationDrone_C|GetSkeletalMeshComponentBudgeted (Variables|DONOTEDIT|GetDrone)) (Variables|DRONE|GetDrone%20VFX%20Attach%20Point) "0, 0, 0" "0, 0, 0" "KeepRelativeOffset" true false))
  (Niagara|SetNiagaraSkeletalMeshComponent _returnvalue "SkeletalMesh" (Variables|/Script/Engine.SkeletalMeshActor|GetSkeletalMeshComponent (Variables|DONOTEDIT|GetMannequin)))
  (CallFunction|SetAdditionalNiagaraParametersOnDroneFX _returnvalue)
  (Components|Activation|Activate _returnvalue false))

(fn SpawnTeleportationDrone
  (Variables|DONOTEDIT|SetDrone 0)
  (if (Variables|DRONE|GetSpawnDrone)
    (bind _returnvalue (Game|SpawnActorfromClass self (Variables|DRONE|GetTeleportation%20Drone) (Transformation|GetWorldTransform (Variables|/Script/Engine.SkeletalMeshActor|GetSkeletalMeshComponent (Variables|DONOTEDIT|GetMannequin))) "AlwaysSpawn" "SelectDefaultAtRuntime" self false (not (GlobalUIContext|IsinZone (BlueprintContext|GetFortGlobalUIContext))) 0))
    (bind _output_get (Variables|DONOTEDIT|SetDrone _returnvalue))
    (Actor|SetLifeSpan _output_get (Variables|DRONE|GetTeleport%20Bot%20Lifespan))
    (Variables|/Game/Blueprints/Adventures/SpawnTeleporter/BP_TeleportationDrone.BP_TeleportationDrone_C|SetAnimPlayRate (Variables|DRONE|GetTeleport_Bot_AnimPlayRate) _output_get)
    (else
      _))
  (return))

(fn SpawnTeleportationLight
  (if (Variables|DEBUG|GetDEBUG_REMOVESPAWNEDPOINTLIGHT)
    _
    (elif (Fortnite|IsMobilePlatform)
      (Development|PrintString (Variables|DONOTEDIT|GetMannequin) "Rez Out - MOBILE PLATFORM - No Point LIght for you.")
      (else
        (Variables|POINTLIGHT|GetTeleportation%20Point%20Light
          (:then)
          (:else
            (bind _returnvalue (AddComponent|NODE_AddPointLightComponent-0 :RelativeTransform (Math|Conversions|ToTransform(Vector) (+ (Transformation|GetSocketLocation (Variables|/Script/Engine.SkeletalMeshActor|GetSkeletalMeshComponent (Variables|DONOTEDIT|GetMannequin)) "root") (Variables|POINTLIGHT|GetTeleportation%20Light%20Offset)))))
            (bind _output_get (Variables|POINTLIGHT|SetTeleportation%20Point%20Light _returnvalue))
            (Rendering|Components|Light|SetIntensity _output_get 0.0)
            (Rendering|Components|Light|SetLightColor _output_get (Variables|POINTLIGHT|GetTeleportation%20Light%20Color) true)))))))

(fn ConstructionScript)
