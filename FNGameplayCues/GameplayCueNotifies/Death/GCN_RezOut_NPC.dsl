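; Gameplay cue notify for the rez-out death effect on NPCs, extending a parent cue with NPC-specific color and sound.
; On Pre-Sequence Change it runs the parent logic, then loops over all pawn materials, setting the Blue, DeRezEdgeGlow and GridColor parameters to cyan-blue tints on the dynamic instances.
; It then spawns the NPC death sound at the pawn's location.
; The construction script only calls the parent.
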(event EventPre-SequenceChange
  (|CallParentFunction)
  (for _array_element (Variables|/Game/Athena/PlayerPawn_Athena_Generic_Parent.PlayerPawn_Athena_Generic_Parent_C|GetPawnMaterials_ALL (Variables|DONOTEDIT|GetPawn))
    (Rendering|Material|SetVectorParameterValue _array_element "Blue" "(R=0.082682,G=1.168436,B=1.500000,A=1.000000)" :!class /Script/Engine.MaterialInstanceDynamic)
    (Rendering|Material|SetVectorParameterValue _array_element "DeRezEdgeGlow" "(R=0.250126,G=4.701368,B=5.000000,A=1.000000)" :!class /Script/Engine.MaterialInstanceDynamic)
    (Rendering|Material|SetVectorParameterValue _array_element "GridColor" "(R=0.000000,G=0.821185,B=1.000000,A=1.000000)" :!class /Script/Engine.MaterialInstanceDynamic))
  (Audio|SpawnSoundatLocation 0 (Variables|Default|GetSoundonNPCDeath) (Transformation|GetActorLocation (Variables|DONOTEDIT|GetPawn))))

(fn ConstructionScript
  (|CallParentFunction))
