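; GCN_RezOut_NPC_Mannequin is a death gameplay cue notify for the rez-out effect on an NPC mannequin.
; On the Pre-SequenceChange event it runs the parent logic, then plays the NPC death sound at the mannequin actor's location.
; The construction script only calls its parent, so the rest of the effect comes from the parent class.
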(event EventPre-SequenceChange
  (|CallParentFunction)
  (Audio|SpawnSoundatLocation 0 (Variables|Default|GetSoundonNPCDeath) (Transformation|GetActorLocation (Variables|DONOTEDIT|GetMannequin))))

(fn ConstructionScript
  (|CallParentFunction))
