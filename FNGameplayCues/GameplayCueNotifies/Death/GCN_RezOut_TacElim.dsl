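; Gameplay cue notify for the TacElim rez-out death effect on a Fortnite pawn, extending a parent death cue.
; It feeds hit location, momentum, headshot flag and normalized last-update velocity into the spawned Niagara effect, also printing that velocity.
; On burst, it plays de-rez audio at the target when in zone, with a game-state mutator check.
; An elimination montage function selects between base and DBNO montages using the headshot flag.
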(event EventSetAdditionalNiagaraParametersonSpawnedFX
  (params (Spawned%20Death%20FX :type Niagara%20Particle%20System%20Component%20Object%20Reference))
  (bind _pawn (Variables|DONOTEDIT|GetPawn))
  (bind _charactermovement (Variables|/Script/Engine.Character|GetCharacterMovement _pawn))
  (bind _returnvalue (Pawn|Components|CharacterMovement|GetLastUpdateVelocity _charactermovement))
  (bind _returnvalue_2 (Math|Vector|Normalize _returnvalue))
  (|CallParentFunction Spawned%20Death%20FX)
  (Niagara|SetNiagaraVariable(Vector3) Spawned%20Death%20FX "ImpactLocation" (Variables|OnDeathParameters|GetOn%20Death%20Hit%20Location))
  (Niagara|SetNiagaraVariable(Vector3) Spawned%20Death%20FX "LinearForceVector" (Variables|OnDeathParameters|GetOn%20Death%20Hit%20Momentum))
  (Niagara|SetNiagaraVariableByString(Bool) Spawned%20Death%20FX "Headshot?" (Variables|OnDeathParameters|GetbisHeadshot))
  (Niagara|SetNiagaraVariable(Vector3) Spawned%20Death%20FX "PlayerVelocity" _returnvalue_2)
  (Development|PrintString 0 (Utilities|String|ToString(Vector) _returnvalue_2)))

(event EventOnBurst(EitherVFX-preferthisversion)
  (params (MyTarget :type Actor%20Object%20Reference) (Parameters :type Gameplay%20Cue%20Parameters%20Structure%20%28by%20ref%29) (ParticleComponents :type Array%20of%20FXSystem%20Component%20Object%20References) (AudioComponents :type Array%20of%20Audio%20Component%20Object%20References) (BurstCameraShakeInstance :type Camera%20Shake%20Base%20Object%20Reference) (BurstDecalInstance :type Decal%20Actor%20Object%20Reference))
  (if (GlobalUIContext|IsinZone (BlueprintContext|GetFortGlobalUIContext))
    (CallFunction|PlayDeRezAudioatLocation MyTarget))
  (|CallParentFunction MyTarget Parameters ParticleComponents AudioComponents BurstCameraShakeInstance BurstDecalInstance))

(fn PlayElminationAnimMontage
  (Animation|PlayLocalAnimMontage (Variables|DONOTEDIT|GetPawn) (select (Variables|OnDeathParameters|GetbisHeadshot) (Variables|ELIMINATIONANIMS|GetBase%20Elimination%20Montage) (Variables|ELIMINATIONANIMS|GetDBNO%20Elimination%20Montage)) 1.0 "None" false)
  (if (Variables|DEBUG|GetDEBUG_ANIMATION)
    (Development|PrintText (Variables|DONOTEDIT|GetPawn) (Utilities|Text|FormatText "NSLOCTEXT(\"[9D98346552C76AF16288394E7F834B49]\", \"7337F1824B8C6EC6A48EB2A7DEF7FEB8\", \"PlayLocalAnimMontage IsDBNO{is} WasDBNO{was} Swimming{swimming} Montage{montage}\")" (Utilities|Text|ToText(Boolean) (Variables|/Script/FortniteGame.FortPawn|GetbIsDBNO (Variables|DONOTEDIT|GetPawn))) (Utilities|Text|ToText(Boolean) (Health|WasDBNOOnDeath (Variables|DONOTEDIT|GetPawn))) (Utilities|Text|ToText(Boolean) (WaterBody|IsSurfaceSwimming (Variables|DONOTEDIT|GetPawn))) (Utilities|Text|ToText(Object) 0)))))

(fn PlayDeRezAudioAtLocation
  (params (MyTarget :type Actor%20Object%20Reference))
  (bind _returnvalue (Fortnite|Athena|GameState|GetGameStateAthenafromContext MyTarget))
  (Utilities|IsValid _returnvalue
    (:Is%20Valid
      (if (Mutators|ShouldPlayDeathSoundEffects _returnvalue MyTarget)
        _))
    (:Is%20Not%20Valid
      _))
  (Audio|PlaySoundatLocation MyTarget "/FNGameplayCues/Sounds/Feral/SC_TacS_Kill_DeRez.SC_TacS_Kill_DeRez" (Transformation|GetActorLocation MyTarget) "0, 0, 0" 1.0 1.0 0.0 0 0 MyTarget))

(fn ConstructionScript
  (|CallParentFunction))
