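; Gameplay cue notify for weapon rez-out, a glow and dissolve effect on the target weapon's meshes, filed under death cues.
; On burst it gathers the meshes to dissolve, sets timeline play rates and starts the glow timeline.
; An optional gauntlet mode instead copies the pawn's cosmetic mesh, materials, pose and transform onto a poseable mesh.
; Helpers apply dissolve materials and spawn attached Niagara VFX; animation and audio functions are empty placeholders.
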(event EventOnBurst(EitherVFX-preferthisversion)
  (params (MyTarget :type Actor%20Object%20Reference) (Parameters :type Gameplay%20Cue%20Parameters%20Structure%20%28by%20ref%29) (ParticleComponents :type Array%20of%20FXSystem%20Component%20Object%20References) (AudioComponents :type Array%20of%20Audio%20Component%20Object%20References) (BurstCameraShakeInstance :type Camera%20Shake%20Base%20Object%20Reference) (BurstDecalInstance :type Decal%20Actor%20Object%20Reference))
  (bind (_normalizedmagnitude _rawmagnitude _effectcontext _matchedtagname _originaltag _aggregatedsourcetags _aggregatedtargettags _location _normal _instigator) (Ability|GameplayCue|BreakGameplayCueParameters Parameters))
  (bind _gauntletpartposeablemesh (Variables|Default|GetGauntletPartPoseableMesh))
  (Variables|DONOTEDIT|SetParticle%20Components ParticleComponents)
  (Variables|DONOTEDIT|SetPawn 0)
  (Variables|Default|SetWeapon 0)
  (Utilities|Set|Clear (Variables|DONOTEDIT|GetDissolve))
  (Development|SetHiddeninGame (Variables|Default|GetGauntletPartPoseableMesh) (not (Variables|Default|GetbUseGauntletPartMesh)) false)
  (bind _asfort_weapon (Utilities|Casting|CastToFortWeapon MyTarget)
    (:then
      (Variables|Default|SetWeapon _asfort_weapon)
      (bind _asplayer_pawn_athena (Utilities|Casting|CastToPlayerPawn_Athena _instigator)
        (:then
          (bind _output_get (Variables|DONOTEDIT|SetPawn _asplayer_pawn_athena))
          (if (Variables|Default|GetbUseGauntletPartMesh)
            (Utilities|FlowControl|Sequence
              (:then_0
                (Utilities|FlowControl|DoOnce false
                  (:Completed
                    (Utilities|IsValid (stale Custom|PlayerPawnGauntletMeshPartReady MeshPart)
                      (:Is%20Valid
                        (Components|SkinnedMesh|SetSkinnedAssetandUpdate _gauntletpartposeablemesh (Variables|/Script/Engine.SkinnedMeshComponent|GetSkeletalMesh (stale Custom|PlayerPawnGauntletMeshPartReady MeshPart)) true)
                        (CallFunction|CopyMaterialstoPoseableMesh (stale Custom|PlayerPawnGauntletMeshPartReady MeshPart))
                        (Components|PoseableMesh|CopyPosefromSkeletalComponent _gauntletpartposeablemesh (stale Custom|PlayerPawnGauntletMeshPartReady MeshPart))
                        (Transformation|SetWorldTransform _gauntletpartposeablemesh (Transformation|GetWorldTransform (stale Custom|PlayerPawnGauntletMeshPartReady MeshPart)) false false))))))
              (:then_1
                (Cosmetics|WhenCosmeticSpecificPartApplied _output_get (delegate Custom|PlayerPawnGauntletMeshPartReady) false "Gameplay"))))
          ; Function Placeholders
          (CallFunction|PlayElminationAnimMontage)
          (CallFunction|StopLoopingAudio)
          (CallFunction|Pre-SequenceChange)
          ;[1 1 1] Get all Meshes into one Set
          (Variables|DONOTEDIT|SetMeshes%20to%20Dissolve (select (Variables|Default|GetbUseGauntletPartMesh) (Utilities|Array|MakeArray (Variables|Default|GetGauntletPartPoseableMesh)) (Visual|GetWeaponMeshes (Variables|Default|GetWeapon))))
          (CallFunction|SetTimelinePlayrates)
          (Components|Timeline|PlayfromStart (Variables|GCN_WeaponRezOut|GetTFX_GlowCharacterMesh)))
        (:CastFailed
          (Development|PrintString 0 "CAST FAILED - to Fort Weapon"))))
    (:CastFailed
      (Development|PrintString 0 "CAST FAILED - to Fort Weapon"))))

(event Custom|PlayerPawnGauntletMeshPartReady
  (params (Pawn :type Fort%20Player%20Pawn%20Object%20Reference) (Part :type Custom%20Character%20Part%20Object%20Reference) (MeshPart :type Skeletal%20Mesh%20Component%20Object%20Reference) (PartType :type EFortCustomPartType%20Enum :default "Head"))
  (bind _gauntletpartposeablemesh (Variables|Default|GetGauntletPartPoseableMesh))
  (Utilities|FlowControl|DoOnce false
    (:Completed
      (Utilities|IsValid MeshPart
        (:Is%20Valid
          (Components|SkinnedMesh|SetSkinnedAssetandUpdate _gauntletpartposeablemesh (Variables|/Script/Engine.SkinnedMeshComponent|GetSkeletalMesh MeshPart) true)
          (CallFunction|CopyMaterialstoPoseableMesh MeshPart)
          (Components|PoseableMesh|CopyPosefromSkeletalComponent _gauntletpartposeablemesh MeshPart)
          (Transformation|SetWorldTransform _gauntletpartposeablemesh (Transformation|GetWorldTransform MeshPart) false false)
          ; Function Placeholders
          (CallFunction|PlayElminationAnimMontage)
          (CallFunction|StopLoopingAudio)
          (CallFunction|Pre-SequenceChange)
          ;[1 1 1] Get all Meshes into one Set
          (Variables|DONOTEDIT|SetMeshes%20to%20Dissolve (select (Variables|Default|GetbUseGauntletPartMesh) (Utilities|Array|MakeArray (Variables|Default|GetGauntletPartPoseableMesh)) (Visual|GetWeaponMeshes (Variables|Default|GetWeapon))))
          (CallFunction|SetTimelinePlayrates)
          (Components|Timeline|PlayfromStart (Variables|GCN_WeaponRezOut|GetTFX_GlowCharacterMesh)))))))

(event Custom|Pre-Sequence%20Change)

(event Custom|Set%20Additional%20Niagara%20Parameters%20on%20Spawned%20FX
  (params (Spawned%20Death%20FX :type Niagara%20Particle%20System%20Component%20Object%20Reference)))

(event Custom|Set%20Additional%20Niagara%20Parameters%20on%20Drone%20FX
  (params (Drone%20FX :type Niagara%20Particle%20System%20Component%20Object%20Reference)))

(fn CopyMaterialsToPoseableMesh
  (params (GauntletSkeletalMeshComponent :type Skeletal%20Mesh%20Component%20Object%20Reference))
  (Utilities|IsValid GauntletSkeletalMeshComponent
    (:Is%20Valid
      (for _index (range (Rendering|Material|GetNumMaterials GauntletSkeletalMeshComponent))
        (bind _returnvalue (Rendering|Material|GetMaterial GauntletSkeletalMeshComponent _index))
        (Utilities|IsValid _returnvalue
          (:Is%20Valid
            (Rendering|Material|SetMaterial (Variables|Default|GetGauntletPartPoseableMesh) _index _returnvalue)))))))

(fn DissolveMaterialSetup
  (for _array_element (Variables|DONOTEDIT|GetMeshes%20to%20Dissolve)
    (for _index (range (Rendering|Material|GetNumMaterials _array_element))
      (bind _returnvalue (Rendering|Material|CreateDynamicMaterialInstance _array_element (Variables|DISSOLVE|GetMat%20Chracter%20Dissolve) "None" "None" :!class /Script/Engine.KismetMaterialLibrary))
      (Rendering|Material|SetMaterial _array_element _index _returnvalue)))
  (return))

(fn PlayElminationAnimMontage
  (return))

(fn SetTimelinePlayrates
  (Components|Timeline|SetPlayRate (Variables|GCN_WeaponRezOut|GetTFX_ResOutCharacterMesh) (Variables|DISSOLVE|GetDissolve%20Timeline%20Playrate))
  (Components|Timeline|SetPlayRate (Variables|GCN_WeaponRezOut|GetTFX_GlowCharacterMesh) (Variables|DISSOLVE|GetGlow%20Timeline%20Playrate)))

(fn SpawnDissolveVFX
  (if (Variables|DEBUG|GetDEBUG_REMOVESPAWNVFX)
    (Development|PrintString 0 "DEBUG - Removed Spawned VFX")
    (return)
    (else
      (bind _returnvalue (Niagara|SpawnSystemAttached (Variables|DISSOLVE|GetDissolve%20Visual%20Effect) (select (Variables|Default|GetbUseGauntletPartMesh) (Variables|Default|GetGauntletPartPoseableMesh) (Visual|GetWeaponMesh (Variables|Default|GetWeapon))) (Variables|DISSOLVE|GetDissolve%20VFX%20Spawn%20Point%20Name) "0, 0, 0" "0, 0, 0" "SnapToTarget" true))
      (bind _output_get (Variables|DONOTEDIT|SetSpawned%20Death%20VFX _returnvalue))
      (Niagara|SetNiagaraVariable(Vector3) _output_get "Bounds" (Variables|DISSOLVE|GetDissolve%20VFX%20Dimensions))
      (Niagara|SetNiagaraVariable(Vector3) _output_get "Offset" (Variables|DISSOLVE|GetDissolve%20VFX%20Offset))
      (CallFunction|SetAdditionalNiagaraParametersOnSpawnedFX (Variables|DONOTEDIT|GetSpawned%20Death%20VFX))
      (return))))

(fn StopLoopingAudio
  (return))

(fn ConstructionScript)
