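; Death gameplay cue, named for StW, that fades out a Fortnite player pawn's capsule indirect shadow.
; On Executed, where cosmetic events can run, it caches the pawn's skinned meshes across all custom part types and records the main mesh's starting shadow minimum visibility.
; The CapsuleFadeTL timeline then lerps each cached mesh's capsule indirect shadow minimum visibility from that start to full, removing the shadow.
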(event GameplayCueNotify|EventHandleGameplayCue
  (params (MyTarget :type Actor%20Object%20Reference) (EventType :type EGameplayCueEvent%20Enum :default "OnActive") (Parameters :type Gameplay%20Cue%20Parameters%20Structure%20%28by%20ref%29))
  (switch Utilities|FlowControl|Switch|SwitchonEGameplayCueEvent EventType
    (:Executed
      (Networking|CanExecuteCosmeticEvents
        (:True
          (bind _asfort_player_pawn (Utilities|Casting|CastToFortPlayerPawn MyTarget))
          (CallFunction|SkeletalMeshSetup _asfort_player_pawn)
          (Variables|Default|SetStarting%20Min%20Capsule%20Shadow%20Vis (Variables|/Script/Engine.SkinnedMeshComponent|GetCapsuleIndirectShadowMinVisibility (Variables|/Script/Engine.Character|GetMesh _asfort_player_pawn)))
          (bind (_direction _removeshadow) (Timeline|CapsuleFadeTL PlayFromStart)
            (:Update
              (for _array_element (Variables|Default|GetSkinnedMesh)
                (Lighting|SetCapsuleIndirectShadowMinVisibility _array_element (Math|Float|Lerp (Variables|Default|GetStarting%20Min%20Capsule%20Shadow%20Vis) 1.0 _removeshadow))))))))))

(fn SkeletalMeshSetup
  (params (FortPawn :type Fort%20Player%20Pawn%20Object%20Reference))
  ;[1 1 1] We don't need to recycle the old mesh references
  (Utilities|Array|Clear (Variables|Default|GetSkinnedMesh))
  ;[1 1 1] For each part type, add the mesh from the pawn (or nullptr if the part doesn't exist)
  (bind _enumvalue (Utilities|Enum|ForEachEFortCustomPartType false)
    (:LoopBody
      (for _array_element (Customization|GetSkinnedMeshComponentsforPartType (Variables|GetFortPawn) _enumvalue)
        (Utilities|Array|Add (Variables|Default|GetSkinnedMesh) _array_element)))
    (:then
      ;[1 1 1] Remove any invalid mesh references
      (Utilities|Array|RemoveItem (Variables|Default|GetSkinnedMesh)))))

(fn ConstructionScript)
